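# Post-process style AnimBP for the F_LRG_DodgyOven female body cosmetic, adding jaw correction and zipper physics.
# It copies the parent pose through a Control Rig, then modifies chin and jaw bones, whose transforms the EventGraph samples from the head mesh.
# A weighted constraint anchors the zipper master to clavicles, neck and spine.
# The cached result feeds two AnimDynamics zipper setups, selecting world-space in the front end and root-relative otherwise.
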AnimBlueprint /BRCosmetics/Characters/Player/Female/Large/Bodies/F_LRG_DodgyOven/Meshes/F_LRG_DodgyOven_AnimBP
target_skeleton=&/F_LRG_DodgyOven_Skeleton
graphs [
 AnimGraph {nodes [
   n1 = Anim Dynamics - Bone: dyn_zipper!(ComponentPose=Local To Component(LocalPose=Use cached pose 'StartPose'()), Alpha=1, GravityOverride:Vector="0,0,-980", ExternalForce) -> (Pose) {
    angularDampingOverride=0.999
    angularSpringConstant=500
    bOverrideAngularDamping=true
    bUseGravityOverride=true
    gravityScale=-1
    simulationSpace=RootRelative
    alphaBoolBlend {blendOption=Linear}
    boundBone {boneName=dyn_zipper}
    componentAppliedLinearAccClamp=<5000 5000 5000>
    componentLinearAccScale=<0.3 0.3 0.4>
    componentLinearVelScale=<0.4 0.4 0.5>
    gravityOverride {z=-980}
    physicsBodyDefinitions [{
      sphereCollisionRadius=1
      boundBone {boneName=dyn_zipper}
      boxExtents=<10 5 10>
      constraintSetup {angularLimitsMax=<0 5 150>, angularLimitsMin=<0 -5 -5>, angularTarget=<0 0 0>}
      localJointOffset=<2 0 0>
     }]
   }
   n2 = Anim Dynamics - Bone: dyn_zipper(ComponentPose=Local To Component(LocalPose=Use cached pose 'StartPose'()), Alpha=1, GravityOverride="0,0,-980") {
    angularDampingOverride=0.9998
    angularSpringConstant=500
    bOverrideAngularDamping=true
    bUseGravityOverride=true
    gravityScale=-1
    simulationSpace=World
    alphaBoolBlend {blendOption=Linear}
    boundBone {boneName=dyn_zipper}
    componentAppliedLinearAccClamp=<500 500 100>
    componentLinearAccScale=<0.2 0.2 0.1>
    componentLinearVelScale=<0.3 0.3 0.2>
    gravityOverride {z=-980}
    physicsBodyDefinitions [{
      sphereCollisionRadius=1
      boundBone {boneName=dyn_zipper}
      boxExtents=<10 5 10>
      constraintSetup {angularLimitsMax=<0 5 150>, angularLimitsMin=<0 -5 -5>, angularTarget=<0 0 0>}
      localJointOffset=<0.5 0 0>
     }]
   }
   Output Pose(Result=Blend Poses by bool(BlendPose_0=Component To Local(ComponentPose=@n2), BlendPose_1=Component To Local(ComponentPose=@n1), bActiveValue=Get bIsInFrontEnd()) {blendType=Linear}) {layerGroup=None}
   n3 = Control Rig(Source=Copy Pose From Mesh() {bUseAttachedParent=true}) {
    alphaBoolBlend {blendOption=Linear}
    controlRigAssetReference {blueprintRigClass=&/F_LRG_DodgyOven_CtrlRig.F_LRG_DodgyOven_CtrlRig_C}
   }
   n4 = "Transform (Modify) Bone - Bone: FACIAL_C_UnderChin"!(ComponentPose=Local To Component(LocalPose=@n3), bAlphaBoolEnabled:bool=True, Translation="0,0,0", Rotation="0,0,0") -> (Pose) {
    alphaInputType=Bool
    rotationMode=BMM_Replace
    rotationSpace=BCS_ParentBoneSpace
    translationMode=BMM_Replace
    translationSpace=BCS_ParentBoneSpace
    alphaBoolBlend {blendOption=Linear}
    boneToModify {boneName=FACIAL_C_UnderChin}
   }
   n5 = "Transform (Modify) Bone - Bone: FACIAL_L_UnderChin"!(ComponentPose=@n4, bAlphaBoolEnabled:bool=True, Translation="0,0,0", Rotation="0,0,0") -> (Pose) {
    alphaInputType=Bool
    rotationMode=BMM_Replace
    rotationSpace=BCS_ParentBoneSpace
    translationMode=BMM_Replace
    translationSpace=BCS_ParentBoneSpace
    alphaBoolBlend {blendOption=Linear}
    boneToModify {boneName=FACIAL_L_UnderChin}
   }
   n6 = "Transform (Modify) Bone - Bone: FACIAL_R_UnderChin"!(ComponentPose=@n5, bAlphaBoolEnabled:bool=True, Translation="0,0,0", Rotation="0,0,0") -> (Pose) {
    alphaInputType=Bool
    rotationMode=BMM_Replace
    rotationSpace=BCS_ParentBoneSpace
    translationMode=BMM_Replace
    translationSpace=BCS_ParentBoneSpace
    alphaBoolBlend {blendOption=Linear}
    boneToModify {boneName=FACIAL_R_UnderChin}
   }
   n7 = "Transform (Modify) Bone - Bone: FACIAL_C_Jaw"!(ComponentPose=@n6, bAlphaBoolEnabled:bool=True, Translation="0,0,0", Rotation="0,0,0") -> (Pose) {
    alphaInputType=Bool
    rotationMode=BMM_Replace
    rotationSpace=BCS_ParentBoneSpace
    translationMode=BMM_Replace
    translationSpace=BCS_ParentBoneSpace
    alphaBoolBlend {blendOption=Linear}
    boneToModify {boneName=FACIAL_C_Jaw}
   }
   n8 = "Transform (Modify) Bone - Bone: FACIAL_C_Jawline"!(ComponentPose=@n7, bAlphaBoolEnabled:bool=True, Translation="0,0,0", Rotation="0,0,0") -> (Pose) {
    alphaInputType=Bool
    rotationMode=BMM_Replace
    rotationSpace=BCS_ParentBoneSpace
    translationMode=BMM_Replace
    translationSpace=BCS_ParentBoneSpace
    alphaBoolBlend {blendOption=Linear}
    boneToModify {boneName=FACIAL_C_Jawline}
   }
   n9 = "Transform (Modify) Bone - Bone: FACIAL_L_Jawline"!(ComponentPose=@n8, bAlphaBoolEnabled:bool=True, Translation="0,0,0", Rotation="0,0,0") -> (Pose) {
    alphaInputType=Bool
    rotationMode=BMM_Replace
    rotationSpace=BCS_ParentBoneSpace
    translationMode=BMM_Replace
    translationSpace=BCS_ParentBoneSpace
    alphaBoolBlend {blendOption=Linear}
    boneToModify {boneName=FACIAL_L_Jawline}
   }
   n10 = "Transform (Modify) Bone - Bone: FACIAL_R_Jawline"!(ComponentPose=@n9, bAlphaBoolEnabled:bool=True, Translation="0,0,0", Rotation="0,0,0") -> (Pose) {
    alphaInputType=Bool
    rotationMode=BMM_Replace
    rotationSpace=BCS_ParentBoneSpace
    translationMode=BMM_Replace
    translationSpace=BCS_ParentBoneSpace
    alphaBoolBlend {blendOption=Linear}
    boneToModify {boneName=FACIAL_R_Jawline}
   }
   n11 = "Constraint - dyn_zipper_MSTR (PPPPPP)"!(
    ComponentPose=@n10
    Alpha=1
    ConstraintWeights_0=0.063
    ConstraintWeights_1:float=0.068
    ConstraintWeights_2:float=0.024
    ConstraintWeights_3:float=0.043
    ConstraintWeights_4:float=0.203
    ConstraintWeights_5:float=0.599
   ) -> (Pose) {
    alphaBoolBlend {blendOption=Linear}
    boneToModify {boneName=dyn_zipper_MSTR}
    constraintSetup [
     {targetBone {boneName=clavicle_l}}
     {targetBone {boneName=clavicle_r}}
     {targetBone {boneName=neck_01}}
     {targetBone {boneName=spine_03}}
     {targetBone {boneName=spine_04}}
     {targetBone {boneName=spine_05}}
    ]
    constraintWeights [1,1,1,1,1,1]
   }
   Save cached pose 'StartPose'(Pose=Component To Local(ComponentPose=@n11)) {cacheName=StartPose}
  ]}
 EventGraph {nodes [
   Event BlueprintUpdateAnimation()
   > n1 = Sequence()
   n2 = Gravity Override State Function DEPRECATED(
    execute=@n1.then_1
    CharacterPartAnimInstance=Self-Reference()
    GameTimeInSeconds=GetGameTimeInSeconds()
    GravityOverrideParams=Get Gravity Override Params()
   )
   > Set Gravity Override(Gravity Override=@n2)
   n3 = Get HeadMesh() -> (HeadMesh:Skeletal Mesh Component Budgeted Object Reference)
   Is Valid(exec=@n1.then_0, InputObject=@n3)
   n4 = GetSocketBoneName!(self:Skinned Mesh Component Object Reference=@n3, InSocketName:Name=FACIAL_C_FacialRoot) -> (ReturnValue:Name=None)
   > Set IsFacialBones(IsFacialBones:bool="Not Equal Exactly (String)"(A="To String (Name)"(InName=@n4), B=None)) -> (Output_Get:bool)
   n5 = Get HeadMesh()
   n6 = BreakTransform(InTransform=GetSocketTransform(self=@n5, InSocketName=FACIAL_C_UnderChin, TransformSpace=RTS_ParentBoneSpace))
   > Set FACIAL_C_UnderChin_Translation(FACIAL_C_UnderChin_Translation:Vector=@n6) -> (Output_Get:Vector)
   > Set FACIAL_C_UnderChin_Rotation(FACIAL_C_UnderChin_Rotation:Rotator=@n6.Rotation) -> (Output_Get:Rotator)
   n7 = BreakTransform(InTransform=GetSocketTransform(self=@n5, InSocketName=FACIAL_L_UnderChin, TransformSpace=RTS_ParentBoneSpace))
   > Set FACIAL_L_UnderChin_Translation(FACIAL_L_UnderChin_Translation:Vector=@n7) -> (Output_Get:Vector)
   > Set FACIAL_L_UnderChin_Rotation(FACIAL_L_UnderChin_Rotation:Rotator=@n7.Rotation) -> (Output_Get:Rotator)
   n8 = BreakTransform(InTransform=GetSocketTransform(self=@n5, InSocketName=FACIAL_R_UnderChin, TransformSpace=RTS_ParentBoneSpace))
   > Set FACIAL_R_UnderChin_Translation(FACIAL_R_UnderChin_Translation:Vector=@n8) -> (Output_Get:Vector)
   > Set FACIAL_R_UnderChin_Rotation(FACIAL_R_UnderChin_Rotation:Rotator=@n8.Rotation) -> (Output_Get:Rotator)
   n9 = BreakTransform(InTransform=GetSocketTransform(self=@n5, InSocketName=FACIAL_C_Jaw, TransformSpace=RTS_ParentBoneSpace))
   > Set FACIAL_C_Jaw_Translation(FACIAL_C_Jaw_Translation:Vector=@n9) -> (Output_Get:Vector)
   > Set FACIAL_C_Jaw_Rotation(FACIAL_C_Jaw_Rotation:Rotator=@n9.Rotation) -> (Output_Get:Rotator)
   n10 = BreakTransform(InTransform=GetSocketTransform(self=@n5, InSocketName=FACIAL_C_Jawline, TransformSpace=RTS_ParentBoneSpace))
   > Set FACIAL_C_Jawline_Translation(FACIAL_C_Jawline_Translation:Vector=@n10) -> (Output_Get:Vector)
   > Set FACIAL_C_Jawline_Rotation(FACIAL_C_Jawline_Rotation:Rotator=@n10.Rotation) -> (Output_Get:Rotator)
   n11 = BreakTransform(InTransform=GetSocketTransform(self=@n5, InSocketName=FACIAL_L_Jawline, TransformSpace=RTS_ParentBoneSpace))
   > Set FACIAL_L_Jawline_Translation(FACIAL_L_Jawline_Translation:Vector=@n11) -> (Output_Get:Vector)
   > Set FACIAL_L_Jawline_Rotation(FACIAL_L_Jawline_Rotation:Rotator=@n11.Rotation) -> (Output_Get:Rotator)
   n12 = BreakTransform(InTransform=GetSocketTransform(self=@n5, InSocketName=FACIAL_R_Jawline, TransformSpace=RTS_ParentBoneSpace))
   > Set FACIAL_R_Jawline_Translation(FACIAL_R_Jawline_Translation:Vector=@n12) -> (Output_Get:Vector)
   > Set FACIAL_R_Jawline_Rotation(FACIAL_R_Jawline_Rotation:Rotator=@n12.Rotation) -> (Output_Get:Rotator)
   n13 = Event OnCharacterCustomizationLoaded()
   > Set HeadMesh(HeadMesh:Skeletal Mesh Component Budgeted Object Reference=Get Array Item with EFort Custom Part Type(Array=@n13.CustomMeshParts, PartType=Head)) -> (Output_Get:Skeletal Mesh Component Budgeted Object Reference)
  ]}
]
variables [
 FACIAL_C_Jaw_Rotation {type=Rotator}
 FACIAL_C_Jaw_Translation {type=Vector}
 FACIAL_C_Jawline_Rotation {type=Rotator}
 FACIAL_C_Jawline_Translation {type=Vector}
 FACIAL_C_UnderChin_Rotation {type=Rotator}
 FACIAL_C_UnderChin_Translation {type=Vector}
 FACIAL_L_Jawline_Rotation {type=Rotator}
 FACIAL_L_Jawline_Translation {type=Vector}
 FACIAL_L_UnderChin_Rotation {type=Rotator}
 FACIAL_L_UnderChin_Translation {type=Vector}
 FACIAL_R_Jawline_Rotation {type=Rotator}
 FACIAL_R_Jawline_Translation {type=Vector}
 FACIAL_R_UnderChin_Rotation {type=Rotator}
 FACIAL_R_UnderChin_Translation {type=Vector}
 Gravity Override {type=Vector}
 Gravity Override Params {type=GravityOverrideParamsStruct}
 HeadMesh {type=SkeletalMeshComponentBudgeted}
 IsFacialBones {type=boolean}
]
