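# Physics asset for the hood face-accessory mesh of the large female DodgyOven player cosmetic body.
# Capsule bodies cover three dynamic hood bones (front left, front right, upper), the head and the upper spine.
# Each hood bone is constrained directly to the head, with zero-degree angle settings on the front pair and a few degrees on the upper one.
# The head-to-spine constraint locks swing and twist.
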PhysicsAsset /BRCosmetics/Characters/Player/Female/Large/Bodies/F_LRG_DodgyOven/Meshes/Parts/F_LRG_DodgyOven_FaceAcc_Hood_Physics
bodies {
 dyn_hood_fr_l [{shapeType=Capsule, length=4, radius=3, shapeName=dyn_hood_fr_l_capsule, center=<0 2 2>, rotation=<rot 90 -90 0>}]
 dyn_hood_fr_r [{shapeType=Capsule, length=4, radius=3, shapeName=dyn_hood_fr_r_capsule, center=<0 2 2>, rotation=<rot 90 -90 0>}]
 dyn_hood_up [{shapeType=Capsule, length=4, radius=3, shapeName=dyn_hood_up_capsule, center=<0 5 0>, rotation=<rot 0 0 90>}]
 head [{shapeType=Capsule, length=1, radius=1, shapeName=head_capsule}]
 spine_05 [{shapeType=Capsule, length=14.3696, radius=12.2693, shapeName=spine_05_capsule, center=<3.1858 0 0>}]
}
constraints [
 {bone1Name=dyn_hood_up, bone2Name=head, swing1LimitDegrees=2, swing2LimitDegrees=2, twistLimitDegrees=3}
 {
  bone1Name=head
  bone2Name=spine_05
  swing1LimitDegrees=45
  swing1Motion=Locked
  swing2LimitDegrees=45
  swing2Motion=Locked
  twistLimitDegrees=45
  twistMotion=Locked
 }
 {bone1Name=dyn_hood_fr_l, bone2Name=head, swing1LimitDegrees=0, swing2LimitDegrees=0, twistLimitDegrees=0}
 {bone1Name=dyn_hood_fr_r, bone2Name=head, swing1LimitDegrees=0, swing2LimitDegrees=0, twistLimitDegrees=0}
]
