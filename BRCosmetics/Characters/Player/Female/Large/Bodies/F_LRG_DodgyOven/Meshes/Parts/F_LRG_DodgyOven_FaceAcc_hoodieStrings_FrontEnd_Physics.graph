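# Physics asset for the front hoodie-string ends of the DodgyOven large female body, letting the laces dangle and swing against the chest.
# Each side has a three-link capsule lace chain hung from a master bone, with limited swing and very little twist.
# Both master bones and four chest capsules are rotation-locked to spine_03, forming a torso volume that likely serves as collision for the laces.
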PhysicsAsset /BRCosmetics/Characters/Player/Female/Large/Bodies/F_LRG_DodgyOven/Meshes/Parts/F_LRG_DodgyOven_FaceAcc_hoodieStrings_FrontEnd_Physics
bodies {
 dyn_lace_01_l [{shapeType=Capsule, length=2, radius=1.5, shapeName=dyn_lace_01_l_capsule, center=<1.3 0 0>}]
 dyn_lace_01_l_MSTR [{
   shapeType=Capsule
   length=1.6588
   radius=0.6198
   shapeName=dyn_lace_01_l_MSTR_capsule
   center=<1.4877 -0.027 -0.0901>
   rotation=<rot 0 95.3404 -81.5422>
  }]
 dyn_lace_01_r [{shapeType=Capsule, length=2, radius=1.5, shapeName=dyn_lace_01_r_capsule, center=<2 0 0>}]
 dyn_lace_01_r_MSTR [{
   shapeType=Capsule
   length=1.6355
   radius=0.6141
   shapeName=dyn_lace_01_r_MSTR_capsule
   center=<1.4693 0.0409 0.0791>
   rotation=<rot 0 -87.8512 -86.9572>
  }]
 dyn_lace_02_l [{shapeType=Capsule, length=2.2, radius=1.3, shapeName=dyn_lace_02_l_capsule, center=<1 0.1 0>}]
 dyn_lace_02_r [{shapeType=Capsule, length=1.7, radius=1.5, shapeName=dyn_lace_02_r_capsule, center=<2 -0.4 0>}]
 dyn_lace_03_l [{shapeType=Capsule, length=2.2, radius=1.3, shapeName=dyn_lace_03_l_capsule, center=<2.1433 0 0>}]
 dyn_lace_03_r [{shapeType=Capsule, length=2.3, radius=1.2, shapeName=dyn_lace_03_r_capsule, center=<3.3373 -0.5 0>}]
 phys_chest_01_l [{shapeType=Capsule, length=6.2, radius=6, shapeName=phys_chest_01_l_capsule, center=<0 -5.8022 -3.1541>, rotation=<rot 90 0 0>}]
 phys_chest_01_r [{shapeType=Capsule, length=6.2, radius=6, shapeName=phys_chest_01_l_capsule, center=<0 -4.9877 -3.5636>, rotation=<rot 90 0 0>}]
 phys_chest_02_l [{shapeType=Capsule, length=6.2, radius=6, shapeName=phys_chest_02_l_capsule, center=<0 -2.8813 -5.462>, rotation=<rot 0 0 -45.3983>}]
 phys_chest_02_r [{shapeType=Capsule, length=6.2, radius=6, shapeName=phys_chest_02_l_capsule, center=<0 -3.06 -5.4769>, rotation=<rot 0 0 -44.9461>}]
 spine_03 [{shapeType=Capsule, length=6, radius=12, shapeName=spine_03_capsule, center=<8.5168 3.0478 0>, rotation=<rot 0 -90 0>}]
}
constraints [
 {bone1Name=dyn_lace_01_l, bone2Name=dyn_lace_01_l_MSTR, swing1LimitDegrees=30, swing2LimitDegrees=30, twistLimitDegrees=2.5}
 {bone1Name=dyn_lace_02_l, bone2Name=dyn_lace_01_l, swing1LimitDegrees=30, swing2LimitDegrees=30, twistLimitDegrees=2.5}
 {bone1Name=dyn_lace_03_l, bone2Name=dyn_lace_02_l, swing1LimitDegrees=30, swing2LimitDegrees=30, twistLimitDegrees=2.5}
 {bone1Name=dyn_lace_01_r, bone2Name=dyn_lace_01_r_MSTR, swing1LimitDegrees=30, swing2LimitDegrees=30, twistLimitDegrees=2.5}
 {bone1Name=dyn_lace_02_r, bone2Name=dyn_lace_01_r, swing1LimitDegrees=30, swing2LimitDegrees=30, twistLimitDegrees=2.5}
 {bone1Name=dyn_lace_03_r, bone2Name=dyn_lace_02_r, swing1LimitDegrees=30, swing2LimitDegrees=30, twistLimitDegrees=2.5}
 {
  bone1Name=dyn_lace_01_r_MSTR
  bone2Name=spine_03
  swing1LimitDegrees=45
  swing1Motion=Locked
  swing2LimitDegrees=45
  swing2Motion=Locked
  twistLimitDegrees=45
  twistMotion=Locked
 }
 {
  bone1Name=dyn_lace_01_l_MSTR
  bone2Name=spine_03
  swing1LimitDegrees=45
  swing1Motion=Locked
  swing2LimitDegrees=45
  swing2Motion=Locked
  twistLimitDegrees=45
  twistMotion=Locked
 }
 {
  bone1Name=phys_chest_01_l
  bone2Name=spine_03
  swing1LimitDegrees=45
  swing1Motion=Locked
  swing2LimitDegrees=45
  swing2Motion=Locked
  twistLimitDegrees=45
  twistMotion=Locked
 }
 {
  bone1Name=phys_chest_01_r
  bone2Name=spine_03
  swing1LimitDegrees=45
  swing1Motion=Locked
  swing2LimitDegrees=45
  swing2Motion=Locked
  twistLimitDegrees=45
  twistMotion=Locked
 }
 {
  bone1Name=phys_chest_02_l
  bone2Name=spine_03
  swing1LimitDegrees=45
  swing1Motion=Locked
  swing2LimitDegrees=45
  swing2Motion=Locked
  twistLimitDegrees=45
  twistMotion=Locked
 }
 {
  bone1Name=phys_chest_02_r
  bone2Name=spine_03
  swing1LimitDegrees=45
  swing1Motion=Locked
  swing2LimitDegrees=45
  swing2Motion=Locked
  twistLimitDegrees=45
  twistMotion=Locked
 }
]
