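# Post-process style AnimBP for the DodgyOven face accessory, adding secondary motion to bangs, hair locks, hoodie strings, hood and ears.
# It copies the parent pose, constrains lace and chest bones, then blends front-end (world-space) or in-game (root-relative) AnimDynamics/RigidBody chains via bIsInFrontEnd.
# Shared hood RigidBody and ear AnimDynamics follow; separate lace AnimDynamics chains are disconnected from output.
# EventGraph updates gravity/RBAN settings and derives lace offsets from spine yaw.
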AnimBlueprint /BRCosmetics/Characters/Player/Female/Large/Bodies/F_LRG_DodgyOven/Meshes/Parts/F_LRG_DodgyOven_FaceAcc_AnimBP
target_skeleton=&/F_LRG_DodgyOven_FaceAcc_Skeleton
graphs [
 AnimGraph {nodes [
   n1 = Anim Dynamics - Bone: dyn_lace_01_l!(ComponentPose, Alpha=1, GravityOverride:Vector="0,0,0", ExternalForce="0,0,0") -> (Pose) {
    angularDampingOverride=1
    angularSpringConstant=1000
    bAngularSpring=true
    bOverrideAngularDamping=true
    bUseGravityOverride=true
    simulationSpace=World
    alphaBoolBlend {blendOption=Linear}
    boundBone {boneName=dyn_lace_01_l}
    componentAppliedLinearAccClamp=<100 100 100>
    componentLinearAccScale=<0.1 0.1 0.1>
    componentLinearVelScale=<0.2 0.2 0.2>
    physicsBodyDefinitions [{
      boundBone {boneName=dyn_lace_01_l}
      boxExtents=<12 12 12>
      constraintSetup {angularLimitsMax=<5 10 0>, angularLimitsMin=<-5 -10 -50>, angularTarget=<1 0 0>}
      localJointOffset=<0.5 0 0>
     }]
   }
   n2 = Anim Dynamics - Bone: dyn_lace_02_l!(ComponentPose=@n1, Alpha=1, GravityOverride:Vector="0,0,0", ExternalForce="0,0,0") -> (Pose) {
    angularDampingOverride=0.9999
    angularSpringConstant=1000
    bAngularSpring=true
    bOverrideAngularDamping=true
    bUseGravityOverride=true
    simulationSpace=World
    alphaBoolBlend {blendOption=Linear}
    boundBone {boneName=dyn_lace_02_l}
    componentAppliedLinearAccClamp=<100 100 100>
    componentLinearAccScale=<0.1 0.1 0.1>
    componentLinearVelScale=<0.2 0.2 0.2>
    physicsBodyDefinitions [{
      boundBone {boneName=dyn_lace_02_l}
      boxExtents=<10 10 10>
      constraintSetup {angularLimitsMax=<5 10 0>, angularLimitsMin=<-5 -10 -50>, angularTarget=<1 0 0>}
      localJointOffset=<0.5 0 0>
     }]
   }
   n3 = Anim Dynamics - Bone: dyn_lace_03_l!(ComponentPose=@n2, Alpha=1, GravityOverride:Vector="0,0,0", ExternalForce="0,0,0") -> (Pose) {
    angularDampingOverride=0.999
    angularSpringConstant=1000
    bAngularSpring=true
    bOverrideAngularDamping=true
    bUseGravityOverride=true
    simulationSpace=World
    alphaBoolBlend {blendOption=Linear}
    boundBone {boneName=dyn_lace_03_l}
    componentAppliedLinearAccClamp=<100 100 100>
    componentLinearAccScale=<0.1 0.1 0.1>
    componentLinearVelScale=<0.2 0.2 0.2>
    physicsBodyDefinitions [{
      boundBone {boneName=dyn_lace_03_l}
      boxExtents=<8 8 8>
      constraintSetup {angularLimitsMax=<5 10 0>, angularLimitsMin=<-5 -10 -50>, angularTarget=<1 0 0>}
      localJointOffset=<0.5 0 0>
     }]
   }
   n4 = Anim Dynamics - Bone: dyn_lace_01_r!(ComponentPose=@n3, Alpha=1, GravityOverride:Vector="0,0,0", ExternalForce="0,0,0") -> (Pose) {
    angularDampingOverride=1
    angularSpringConstant=1000
    bAngularSpring=true
    bOverrideAngularDamping=true
    bUseGravityOverride=true
    simulationSpace=World
    alphaBoolBlend {blendOption=Linear}
    boundBone {boneName=dyn_lace_01_r}
    componentAppliedLinearAccClamp=<10 10 10>
    componentLinearAccScale=<0.1 0.1 0.1>
    componentLinearVelScale=<0.2 0.2 0.2>
    physicsBodyDefinitions [{
      boundBone {boneName=dyn_lace_01_r}
      boxExtents=<16 16 16>
      constraintSetup {angularLimitsMax=<5 10 0>, angularLimitsMin=<-5 -10 -50>, angularTarget=<1 0 0>}
      localJointOffset=<0.3 0 0>
     }]
   }
   n5 = Anim Dynamics - Bone: dyn_lace_02_r!(ComponentPose=@n4, Alpha=1, GravityOverride:Vector="0,0,0", ExternalForce="0,0,0") -> (Pose) {
    angularDampingOverride=0.9999
    angularSpringConstant=1000
    bAngularSpring=true
    bOverrideAngularDamping=true
    bUseGravityOverride=true
    simulationSpace=World
    alphaBoolBlend {blendOption=Linear}
    boundBone {boneName=dyn_lace_02_r}
    componentAppliedLinearAccClamp=<10 10 10>
    componentLinearAccScale=<0.1 0.1 0.1>
    componentLinearVelScale=<0.2 0.2 0.2>
    physicsBodyDefinitions [{
      boundBone {boneName=dyn_lace_02_r}
      boxExtents=<14 14 14>
      constraintSetup {angularLimitsMax=<5 10 0>, angularLimitsMin=<-5 -10 -50>, angularTarget=<1 0 0>}
      localJointOffset=<0.3 0 0>
     }]
   }
   Anim Dynamics - Bone: dyn_lace_03_r!(ComponentPose=@n5, Alpha=1, GravityOverride:Vector="0,0,0", ExternalForce="0,0,0") -> (Pose) {
    angularDampingOverride=0.999
    angularSpringConstant=1000
    bAngularSpring=true
    bOverrideAngularDamping=true
    bUseGravityOverride=true
    simulationSpace=World
    alphaBoolBlend {blendOption=Linear}
    boundBone {boneName=dyn_lace_03_r}
    componentAppliedLinearAccClamp=<10 10 10>
    componentLinearAccScale=<0.1 0.1 0.1>
    componentLinearVelScale=<0.2 0.2 0.2>
    physicsBodyDefinitions [{
      boundBone {boneName=dyn_lace_03_r}
      boxExtents=<12 12 12>
      constraintSetup {angularLimitsMax=<5 10 0>, angularLimitsMin=<-5 -10 -50>, angularTarget=<1 0 0>}
      localJointOffset=<0.3 0 0>
     }]
   }
   n6 = "Transform (Modify) Bone - Bone: dyn_lace_02_l"!(ComponentPose, bAlphaBoolEnabled:bool=True, Translation="0,0,0", Rotation="0,0,0") -> (Pose) {
    alphaInputType=Bool
    rotationMode=BMM_Additive
    translationMode=BMM_Additive
    alphaBoolBlend {blendOption=Linear}
    boneToModify {boneName=dyn_lace_02_l}
   }
   "Transform (Modify) Bone - Bone: dyn_lace_02_r"!(ComponentPose=@n6, bAlphaBoolEnabled:bool=True, Translation="0,0,0", Rotation="0,0,0") -> (Pose) {
    alphaInputType=Bool
    rotationMode=BMM_Additive
    translationMode=BMM_Additive
    alphaBoolBlend {blendOption=Linear}
    boneToModify {boneName=dyn_lace_02_r}
   }
   n7 = Anim Dynamics - Bone: dyn_lace_01_l(Alpha=1, GravityOverride="0,0,0", ExternalForce="0,0,0") {
    angularDampingOverride=0.9999
    angularSpringConstant=300
    bAngularSpring=true
    bOverrideAngularDamping=true
    bUseGravityOverride=true
    simulationSpace=RootRelative
    alphaBoolBlend {blendOption=Linear}
    boundBone {boneName=dyn_lace_01_l}
    componentAppliedLinearAccClamp=<6000 6000 5000>
    componentLinearAccScale=<0.4 0.4 0.6>
    componentLinearVelScale=<0.5 0.5 0.7>
    physicsBodyDefinitions [{
      boundBone {boneName=dyn_lace_01_l}
      boxExtents=<14 14 14>
      constraintSetup {angularLimitsMax=<10 20 0>, angularLimitsMin=<-10 -20 -50>, angularTarget=<1 0 0>}
      localJointOffset=<1 0 0>
     }]
   }
   n8 = Anim Dynamics - Bone: dyn_lace_02_l(ComponentPose=@n7, Alpha=1, GravityOverride="0,0,0", ExternalForce="0,0,0") {
    angularDampingOverride=0.999
    angularSpringConstant=200
    bAngularSpring=true
    bOverrideAngularDamping=true
    bUseGravityOverride=true
    simulationSpace=RootRelative
    alphaBoolBlend {blendOption=Linear}
    boundBone {boneName=dyn_lace_02_l}
    componentAppliedLinearAccClamp=<5500 5500 5000>
    componentLinearAccScale=<0.4 0.4 0.6>
    componentLinearVelScale=<0.5 0.5 0.7>
    physicsBodyDefinitions [{
      boundBone {boneName=dyn_lace_02_l}
      boxExtents=<12 12 12>
      constraintSetup {angularLimitsMax=<10 20 0>, angularLimitsMin=<-10 -20 -50>, angularTarget=<1 0 0>}
      localJointOffset=<1 0 0>
     }]
   }
   n9 = Anim Dynamics - Bone: dyn_lace_03_l(ComponentPose=@n8, Alpha=1, GravityOverride="0,0,0", ExternalForce="0,0,0") {
    angularDampingOverride=0.99
    angularSpringConstant=100
    bAngularSpring=true
    bOverrideAngularDamping=true
    bUseGravityOverride=true
    simulationSpace=RootRelative
    alphaBoolBlend {blendOption=Linear}
    boundBone {boneName=dyn_lace_03_l}
    componentAppliedLinearAccClamp=<5000 5000 5000>
    componentLinearAccScale=<0.4 0.4 0.6>
    componentLinearVelScale=<0.5 0.5 0.7>
    physicsBodyDefinitions [{
      boundBone {boneName=dyn_lace_03_l}
      boxExtents=<10 10 10>
      constraintSetup {angularLimitsMax=<10 20 0>, angularLimitsMin=<-10 -20 -50>, angularTarget=<1 0 0>}
      localJointOffset=<1 0 0>
     }]
   }
   n10 = Anim Dynamics - Bone: dyn_lace_01_r(ComponentPose=@n9, Alpha=1, GravityOverride="0,0,0", ExternalForce="0,0,0") {
    angularDampingOverride=1
    angularSpringConstant=450
    bAngularSpring=true
    bOverrideAngularDamping=true
    bUseGravityOverride=true
    simulationSpace=RootRelative
    alphaBoolBlend {blendOption=Linear}
    boundBone {boneName=dyn_lace_01_r}
    componentAppliedLinearAccClamp=<6000 6000 5000>
    componentLinearAccScale=<0.2 0.2 0.6>
    componentLinearVelScale=<0.3 0.3 0.7>
    physicsBodyDefinitions [{
      boundBone {boneName=dyn_lace_01_r}
      boxExtents=<16 16 16>
      constraintSetup {angularLimitsMax=<10 20 0>, angularLimitsMin=<-10 -20 -50>, angularTarget=<1 0 0>}
      localJointOffset=<1.5 0 0>
     }]
   }
   n11 = Anim Dynamics - Bone: dyn_lace_02_r(ComponentPose=@n10, Alpha=1, GravityOverride="0,0,0", ExternalForce="0,0,0") {
    angularDampingOverride=0.9998
    angularSpringConstant=300
    bAngularSpring=true
    bOverrideAngularDamping=true
    bUseGravityOverride=true
    simulationSpace=RootRelative
    alphaBoolBlend {blendOption=Linear}
    boundBone {boneName=dyn_lace_02_r}
    componentAppliedLinearAccClamp=<5500 5500 5000>
    componentLinearAccScale=<0.2 0.2 0.6>
    componentLinearVelScale=<0.3 0.3 0.7>
    physicsBodyDefinitions [{
      boundBone {boneName=dyn_lace_02_r}
      boxExtents=<14 14 14>
      constraintSetup {angularLimitsMax=<10 20 0>, angularLimitsMin=<-10 -20 -50>, angularTarget=<1 0 0>}
      localJointOffset=<1.5 0 0>
     }]
   }
   n12 = Anim Dynamics - Bone: dyn_lace_03_r(ComponentPose=@n11, Alpha=1, GravityOverride="0,0,0", ExternalForce="0,0,0") {
    angularDampingOverride=0.998
    angularSpringConstant=150
    bAngularSpring=true
    bOverrideAngularDamping=true
    bUseGravityOverride=true
    simulationSpace=RootRelative
    alphaBoolBlend {blendOption=Linear}
    boundBone {boneName=dyn_lace_03_r}
    componentAppliedLinearAccClamp=<5000 5000 5000>
    componentLinearAccScale=<0.2 0.2 0.6>
    componentLinearVelScale=<0.3 0.3 0.7>
    physicsBodyDefinitions [{
      boundBone {boneName=dyn_lace_03_r}
      boxExtents=<10 10 10>
      constraintSetup {angularLimitsMax=<10 20 0>, angularLimitsMin=<-10 -20 -50>, angularTarget=<1 0 0>}
      localJointOffset=<1.5 0 0>
     }]
   }
   n13 = "Transform (Modify) Bone - Bone: dyn_lace_02_l"(ComponentPose=@n12, bAlphaBoolEnabled=True, Translation="0,0,0", Rotation="0,0,0") {
    alphaInputType=Bool
    rotationMode=BMM_Additive
    translationMode=BMM_Additive
    alphaBoolBlend {blendOption=Linear}
    boneToModify {boneName=dyn_lace_02_l}
   }
   "Transform (Modify) Bone - Bone: dyn_lace_02_r"(ComponentPose=@n13, bAlphaBoolEnabled=True, Translation="0,0,0", Rotation="0,0,0") {
    alphaInputType=Bool
    rotationMode=BMM_Additive
    translationMode=BMM_Additive
    alphaBoolBlend {blendOption=Linear}
    boneToModify {boneName=dyn_lace_02_r}
   }
   Output Pose(Result=Use cached pose 'DynamicPose'()) {layerGroup=None}
   n14 = "Constraint - dyn_lace_01_l_MSTR (PPPPPPPP)"!(
    ComponentPose=Local To Component(LocalPose=Copy Pose From Mesh() {bUseAttachedParent=true})
    Alpha=1
    ConstraintWeights_0=0.133
    ConstraintWeights_1:float=0.01
    ConstraintWeights_2:float=0.005
    ConstraintWeights_3:float=0.005
    ConstraintWeights_4:float=0.038
    ConstraintWeights_5:float=0.034
    ConstraintWeights_6:float=0.13
    ConstraintWeights_7:float=0.645
   ) -> (Pose) {
    alphaBoolBlend {blendOption=Linear}
    boneToModify {boneName=dyn_lace_01_l_MSTR}
    constraintSetup [
     {targetBone {boneName=clavicle_l}}
     {targetBone {boneName=clavicle_r}}
     {targetBone {boneName=dyn_hood_fr_l}}
     {targetBone {boneName=head}}
     {targetBone {boneName=neck_01}}
     {targetBone {boneName=spine_03}}
     {targetBone {boneName=spine_04}}
     {targetBone {boneName=spine_05}}
    ]
    constraintWeights [1,1,1,1,1,1,1,1]
   }
   n15 = "Constraint - dyn_lace_01_r_MSTR (PPPPPPPP)"!(
    ComponentPose=@n14
    Alpha=1
    ConstraintWeights_0=0.012
    ConstraintWeights_1:float=0.127
    ConstraintWeights_2:float=0.007
    ConstraintWeights_3:float=0.007
    ConstraintWeights_4:float=0.04
    ConstraintWeights_5:float=0.033
    ConstraintWeights_6:float=0.134
    ConstraintWeights_7:float=0.64
   ) -> (Pose) {
    alphaBoolBlend {blendOption=Linear}
    boneToModify {boneName=dyn_lace_01_r_MSTR}
    constraintSetup [
     {targetBone {boneName=clavicle_l}}
     {targetBone {boneName=clavicle_r}}
     {targetBone {boneName=dyn_hood_fr_r}}
     {targetBone {boneName=head}}
     {targetBone {boneName=neck_01}}
     {targetBone {boneName=spine_03}}
     {targetBone {boneName=spine_04}}
     {targetBone {boneName=spine_05}}
    ]
    constraintWeights [1,1,1,1,1,1,1,1]
   }
   n16 = "Constraint - phys_chest_01_l (PPPPPPP)"!(
    ComponentPose=@n15
    Alpha=1
    ConstraintWeights_0=0.07
    ConstraintWeights_1:float=0.01
    ConstraintWeights_2:float=0.118
    ConstraintWeights_3:float=0.075
    ConstraintWeights_4:float=0.301
    ConstraintWeights_5:float=0.118
    ConstraintWeights_6:float=0.308
   ) -> (Pose) {
    alphaBoolBlend {blendOption=Linear}
    boneToModify {boneName=phys_chest_01_l}
    constraintSetup [
     {targetBone {boneName=clavicle_l}}
     {targetBone {boneName=clavicle_r}}
     {targetBone {boneName=deform_pec_l}}
     {targetBone {boneName=spine_02}}
     {targetBone {boneName=spine_03}}
     {targetBone {boneName=spine_04}}
     {targetBone {boneName=spine_05}}
    ]
    constraintWeights [1,1,1,1,1,1,1]
   }
   n17 = "Constraint - phys_chest_01_r (PPPPPPP)"!(
    ComponentPose=@n16
    Alpha=1
    ConstraintWeights_0=0.01
    ConstraintWeights_1:float=0.07
    ConstraintWeights_2:float=0.118
    ConstraintWeights_3:float=0.075
    ConstraintWeights_4:float=0.301
    ConstraintWeights_5:float=0.118
    ConstraintWeights_6:float=0.308
   ) -> (Pose) {
    alphaBoolBlend {blendOption=Linear}
    boneToModify {boneName=phys_chest_01_r}
    constraintSetup [
     {targetBone {boneName=clavicle_l}}
     {targetBone {boneName=clavicle_r}}
     {targetBone {boneName=deform_pec_r}}
     {targetBone {boneName=spine_02}}
     {targetBone {boneName=spine_03}}
     {targetBone {boneName=spine_04}}
     {targetBone {boneName=spine_05}}
    ]
    constraintWeights [1,1,1,1,1,1,1]
   }
   n18 = "Constraint - phys_chest_02_l (PPPPPPPP)"!(
    ComponentPose=@n17
    Alpha=1
    ConstraintWeights_0=0.123
    ConstraintWeights_1:float=0.017
    ConstraintWeights_2:float=0.088
    ConstraintWeights_3:float=0.018
    ConstraintWeights_4:float=0.129
    ConstraintWeights_5:float=0.138
    ConstraintWeights_6:float=0.474
    ConstraintWeights_7:float=0.013
   ) -> (Pose) {
    alphaBoolBlend {blendOption=Linear}
    boneToModify {boneName=phys_chest_02_l}
    constraintSetup [
     {targetBone {boneName=clavicle_l}}
     {targetBone {boneName=clavicle_r}}
     {targetBone {boneName=deform_pec_l}}
     {targetBone {boneName=spine_02}}
     {targetBone {boneName=spine_03}}
     {targetBone {boneName=spine_04}}
     {targetBone {boneName=spine_05}}
     {targetBone {boneName=neck_01}}
    ]
    constraintWeights [1,1,1,1,1,1,1,1]
   }
   n19 = "Constraint - phys_chest_02_r (PPPPPPPP)"!(
    ComponentPose=@n18
    Alpha=1
    ConstraintWeights_0=0.017
    ConstraintWeights_1:float=0.123
    ConstraintWeights_2:float=0.088
    ConstraintWeights_3:float=0.018
    ConstraintWeights_4:float=0.129
    ConstraintWeights_5:float=0.138
    ConstraintWeights_6:float=0.474
    ConstraintWeights_7:float=0.013
   ) -> (Pose) {
    alphaBoolBlend {blendOption=Linear}
    boneToModify {boneName=phys_chest_02_r}
    constraintSetup [
     {targetBone {boneName=clavicle_l}}
     {targetBone {boneName=clavicle_r}}
     {targetBone {boneName=deform_pec_r}}
     {targetBone {boneName=spine_02}}
     {targetBone {boneName=spine_03}}
     {targetBone {boneName=spine_04}}
     {targetBone {boneName=spine_05}}
     {targetBone {boneName=neck_01}}
    ]
    constraintWeights [1,1,1,1,1,1,1,1]
   }
   Save cached pose 'StartPose'(Pose=Component To Local(ComponentPose=@n19)) {cacheName=StartPose}
   n20 = Anim Dynamics - Bone: dyn_bang_A_01_l!(ComponentPose=Local To Component(LocalPose=Use cached pose 'StartPose'()), Alpha=1, GravityOverride:Vector="0,0,0", ExternalForce="0,0,0") -> (Pose) {
    angularDampingOverride=0.998
    angularSpringConstant=250
    bAngularSpring=true
    bOverrideAngularDamping=true
    bUseGravityOverride=true
    simulationSpace=RootRelative
    alphaBoolBlend {blendOption=Linear}
    boundBone {boneName=dyn_bang_A_01_l}
    componentAppliedLinearAccClamp=<1000 1000 1000>
    componentLinearAccScale=<1 1 1>
    componentLinearVelScale=<1 1 1>
    physicsBodyDefinitions [{
      boundBone {boneName=dyn_bang_A_01_l}
      boxExtents=<10 10 10>
      constraintSetup {angularLimitsMax=<0 10 0>, angularLimitsMin=<0 -10 -20>, angularTarget=<1 0 0>}
      localJointOffset=<0.5 0 0>
     }]
   }
   n21 = Anim Dynamics - Bone: dyn_bang_B_01_l!(ComponentPose=@n20, Alpha=1, GravityOverride:Vector="0,0,0", ExternalForce="0,0,0") -> (Pose) {
    angularDampingOverride=0.999
    angularSpringConstant=100
    bAngularSpring=true
    bOverrideAngularDamping=true
    bUseGravityOverride=true
    simulationSpace=RootRelative
    alphaBoolBlend {blendOption=Linear}
    boundBone {boneName=dyn_bang_B_01_l}
    componentAppliedLinearAccClamp=<500 500 500>
    componentLinearAccScale=<1 1 1>
    componentLinearVelScale=<1 1 1>
    physicsBodyDefinitions [{
      boundBone {boneName=dyn_bang_B_01_l}
      boxExtents=<10 10 10>
      constraintSetup {angularLimitsMax=<0 10 0>, angularLimitsMin=<0 -10 -20>, angularTarget=<1 0 0>}
      localJointOffset=<0.5 0.5 0>
     }]
   }
   n22 = Anim Dynamics - Bone: dyn_bang_C_01_l!(ComponentPose=@n21, Alpha=1, GravityOverride:Vector="0,0,0", ExternalForce="0,0,0") -> (Pose) {
    angularDampingOverride=0.998
    angularSpringConstant=100
    bAngularSpring=true
    bOverrideAngularDamping=true
    bUseGravityOverride=true
    simulationSpace=RootRelative
    alphaBoolBlend {blendOption=Linear}
    boundBone {boneName=dyn_bang_C_01_l}
    componentAppliedLinearAccClamp=<700 700 700>
    componentLinearAccScale=<1 1 1>
    componentLinearVelScale=<1 1 1>
    physicsBodyDefinitions [{
      boundBone {boneName=dyn_bang_C_01_l}
      boxExtents=<12 12 12>
      constraintSetup {angularLimitsMax=<2.5 0 0>, angularLimitsMin=<-2.5 -10 -10>, angularTarget=<1 0 0>}
      localJointOffset=<1 0 0>
     }]
   }
   n23 = Anim Dynamics - Bone: dyn_bang_C_02_l!(ComponentPose=@n22, Alpha=1, GravityOverride:Vector="0,0,0", ExternalForce="0,0,0") -> (Pose) {
    angularDampingOverride=0.99
    angularSpringConstant=50
    bAngularSpring=true
    bOverrideAngularDamping=true
    bUseGravityOverride=true
    simulationSpace=RootRelative
    alphaBoolBlend {blendOption=Linear}
    boundBone {boneName=dyn_bang_C_02_l}
    componentAppliedLinearAccClamp=<1000 1000 1000>
    componentLinearAccScale=<1 1 1>
    componentLinearVelScale=<1 1 1>
    physicsBodyDefinitions [{
      boundBone {boneName=dyn_bang_C_02_l}
      boxExtents=<8 8 8>
      constraintSetup {angularLimitsMax=<2.5 0 0>, angularLimitsMin=<-2.5 -10 -10>, angularTarget=<1 0 0>}
      localJointOffset=<1 0 0>
     }]
   }
   n24 = Anim Dynamics - Bone: dyn_bang_A_01_r!(ComponentPose=@n23, Alpha=1, GravityOverride:Vector="0,0,0", ExternalForce="0,0,0") -> (Pose) {
    angularDampingOverride=0.99
    angularSpringConstant=50
    bAngularSpring=true
    bOverrideAngularDamping=true
    bUseGravityOverride=true
    simulationSpace=RootRelative
    alphaBoolBlend {blendOption=Linear}
    boundBone {boneName=dyn_bang_A_01_r}
    componentAppliedLinearAccClamp=<1000 1000 1000>
    componentLinearAccScale=<1 1 1>
    componentLinearVelScale=<1 1 1>
    physicsBodyDefinitions [{
      boundBone {boneName=dyn_bang_A_01_r}
      boxExtents=<10 10 10>
      constraintSetup {angularLimitsMax=<0 10 0>, angularLimitsMin=<0 -10 -40>, angularTarget=<1 0 0>}
      localJointOffset=<1 0 0>
     }]
   }
   n25 = Anim Dynamics - Bone: dyn_bang_B_01_r!(ComponentPose=@n24, Alpha=1, GravityOverride:Vector="0,0,0", ExternalForce="0,0,0") -> (Pose) {
    angularDampingOverride=0.988
    bOverrideAngularDamping=true
    bUseGravityOverride=true
    simulationSpace=RootRelative
    alphaBoolBlend {blendOption=Linear}
    boundBone {boneName=dyn_bang_B_01_r}
    componentAppliedLinearAccClamp=<1500 1500 1500>
    componentLinearAccScale=<1 1 1>
    componentLinearVelScale=<1 1 1>
    physicsBodyDefinitions [{
      boundBone {boneName=dyn_bang_B_01_r}
      boxExtents=<10 10 10>
      constraintSetup {angularLimitsMax=<0 10 0>, angularLimitsMin=<0 -10 -40>, angularTarget=<0 0 0>}
      localJointOffset=<0.8 0 0>
     }]
   }
   n26 = Anim Dynamics - Bone: dyn_bang_C_01_r!(ComponentPose=@n25, Alpha=1, GravityOverride:Vector="0,0,0", ExternalForce="0,0,0") -> (Pose) {
    angularDampingOverride=0.99
    angularSpringConstant=50
    bAngularSpring=true
    bOverrideAngularDamping=true
    bUseGravityOverride=true
    simulationSpace=RootRelative
    alphaBoolBlend {blendOption=Linear}
    boundBone {boneName=dyn_bang_C_01_r}
    componentAppliedLinearAccClamp=<1000 1000 1000>
    componentLinearAccScale=<1 1 1>
    componentLinearVelScale=<1 1 1>
    physicsBodyDefinitions [{
      boundBone {boneName=dyn_bang_C_01_r}
      boxExtents=<10 10 10>
      constraintSetup {angularLimitsMax=<0 10 0>, angularLimitsMin=<0 -10 -40>, angularTarget=<1 0 0>}
      localJointOffset=<1 0 0>
     }]
   }
   n27 = Anim Dynamics - Bone: dyn_hair_lock_l!(ComponentPose=@n26, Alpha=1, GravityOverride:Vector="0,0,0", ExternalForce="0,0,0") -> (Pose) {
    angularDampingOverride=0.999
    angularSpringConstant=1500
    bAngularSpring=true
    bOverrideAngularDamping=true
    bUseGravityOverride=true
    simulationSpace=RootRelative
    alphaBoolBlend {blendOption=Linear}
    boundBone {boneName=dyn_hair_lock_l}
    componentAppliedLinearAccClamp=<500 500 500>
    componentLinearAccScale=<0.4 0.4 0.4>
    componentLinearVelScale=<0.6 0.6 0.6>
    physicsBodyDefinitions [{
      boundBone {boneName=dyn_hair_lock_l}
      boxExtents=<15 25 25>
      constraintSetup {angularLimitsMax=<0 10 7>, angularLimitsMin=<0 -15 0>, angularTarget=<1 0 0>}
      localJointOffset=<1.5 0 0>
     }]
   }
   n28 = Anim Dynamics - Bone: dyn_hair_lock_r!(ComponentPose=@n27, Alpha=1, GravityOverride:Vector="0,0,0", ExternalForce="0,0,0") -> (Pose) {
    angularDampingOverride=0.999
    angularSpringConstant=3000
    bAngularSpring=true
    bOverrideAngularDamping=true
    bUseGravityOverride=true
    simulationSpace=RootRelative
    alphaBoolBlend {blendOption=Linear}
    boundBone {boneName=dyn_hair_lock_r}
    componentAppliedLinearAccClamp=<350 350 350>
    componentLinearAccScale=<0.5 0.5 0.5>
    componentLinearVelScale=<0.7 0.7 0.7>
    physicsBodyDefinitions [{
      boundBone {boneName=dyn_hair_lock_r}
      boxExtents=<20 20 20>
      constraintSetup {angularLimitsMax=<0 10 7>, angularLimitsMin=<0 -8 -5>, angularTarget=<1 0 0>}
      localJointOffset=<1 0 0>
     }]
   }
   n29 = RigidBody!(ComponentPose=@n28, Alpha, OverrideWorldGravity:Vector="0,0,0", ExternalForce) -> (Pose) {
    bClampLinearTranslationLimitToRefPose=true
    bOverrideWorldGravity=true
    overridePhysicsAsset=&/F_LRG_DodgyOven_FaceAcc_hoodieStrings_FrontEnd_Physics
    simulationSpace=BaseBoneSpace
    alphaBoolBlend {blendOption=Linear}
    baseBoneRef {boneName=spine_05}
    simSpaceSettings {dampingAlpha=0, maxAngularAcceleration=16, maxAngularVelocity=8, maxLinearAcceleration=1400, maxLinearVelocity=700, worldAlpha=1}
   }
   n30 = Component To Local(ComponentPose=@n29)
   n31 = Anim Dynamics - Bone: dyn_bang_A_01_l(ComponentPose=Local To Component(LocalPose=Use cached pose 'StartPose'()), Alpha=1, GravityOverride="0,0,0", ExternalForce="0,0,0") {
    angularDampingOverride=0.998
    angularSpringConstant=250
    bAngularSpring=true
    bOverrideAngularDamping=true
    bUseGravityOverride=true
    simulationSpace=World
    alphaBoolBlend {blendOption=Linear}
    boundBone {boneName=dyn_bang_A_01_l}
    componentAppliedLinearAccClamp=<1000 1000 1000>
    componentLinearAccScale=<1 1 1>
    componentLinearVelScale=<1 1 1>
    physicsBodyDefinitions [{
      boundBone {boneName=dyn_bang_A_01_l}
      boxExtents=<10 10 10>
      constraintSetup {angularLimitsMax=<0 10 0>, angularLimitsMin=<0 -10 -20>, angularTarget=<1 0 0>}
      localJointOffset=<0.5 0 0>
     }]
   }
   n32 = Anim Dynamics - Bone: dyn_bang_B_01_l(ComponentPose=@n31, Alpha=1, GravityOverride="0,0,0", ExternalForce="0,0,0") {
    angularDampingOverride=0.999
    angularSpringConstant=100
    bAngularSpring=true
    bOverrideAngularDamping=true
    bUseGravityOverride=true
    simulationSpace=World
    alphaBoolBlend {blendOption=Linear}
    boundBone {boneName=dyn_bang_B_01_l}
    componentAppliedLinearAccClamp=<500 500 500>
    componentLinearAccScale=<1 1 1>
    componentLinearVelScale=<1 1 1>
    physicsBodyDefinitions [{
      boundBone {boneName=dyn_bang_B_01_l}
      boxExtents=<10 10 10>
      constraintSetup {angularLimitsMax=<0 10 0>, angularLimitsMin=<0 -10 -20>, angularTarget=<1 0 0>}
      localJointOffset=<0.5 0.5 0>
     }]
   }
   n33 = Anim Dynamics - Bone: dyn_bang_C_01_l(ComponentPose=@n32, Alpha=1, GravityOverride="0,0,0", ExternalForce="0,0,0") {
    angularDampingOverride=0.998
    angularSpringConstant=100
    bAngularSpring=true
    bOverrideAngularDamping=true
    bUseGravityOverride=true
    simulationSpace=World
    alphaBoolBlend {blendOption=Linear}
    boundBone {boneName=dyn_bang_C_01_l}
    componentAppliedLinearAccClamp=<700 700 700>
    componentLinearAccScale=<1 1 1>
    componentLinearVelScale=<1 1 1>
    physicsBodyDefinitions [{
      boundBone {boneName=dyn_bang_C_01_l}
      boxExtents=<12 12 12>
      constraintSetup {angularLimitsMax=<2.5 0 0>, angularLimitsMin=<-2.5 -10 -10>, angularTarget=<1 0 0>}
      localJointOffset=<1 0 0>
     }]
   }
   n34 = Anim Dynamics - Bone: dyn_bang_C_02_l(ComponentPose=@n33, Alpha=1, GravityOverride="0,0,0", ExternalForce="0,0,0") {
    angularDampingOverride=0.99
    angularSpringConstant=50
    bAngularSpring=true
    bOverrideAngularDamping=true
    bUseGravityOverride=true
    simulationSpace=World
    alphaBoolBlend {blendOption=Linear}
    boundBone {boneName=dyn_bang_C_02_l}
    componentAppliedLinearAccClamp=<1000 1000 1000>
    componentLinearAccScale=<1 1 1>
    componentLinearVelScale=<1 1 1>
    physicsBodyDefinitions [{
      boundBone {boneName=dyn_bang_C_02_l}
      boxExtents=<8 8 8>
      constraintSetup {angularLimitsMax=<2.5 0 0>, angularLimitsMin=<-2.5 -10 -10>, angularTarget=<1 0 0>}
      localJointOffset=<1 0 0>
     }]
   }
   n35 = Anim Dynamics - Bone: dyn_bang_A_01_r(ComponentPose=@n34, Alpha=1, GravityOverride="0,0,0", ExternalForce="0,0,0") {
    angularDampingOverride=0.99
    angularSpringConstant=50
    bAngularSpring=true
    bOverrideAngularDamping=true
    bUseGravityOverride=true
    simulationSpace=World
    alphaBoolBlend {blendOption=Linear}
    boundBone {boneName=dyn_bang_A_01_r}
    componentAppliedLinearAccClamp=<1000 1000 1000>
    componentLinearAccScale=<1 1 1>
    componentLinearVelScale=<1 1 1>
    physicsBodyDefinitions [{
      boundBone {boneName=dyn_bang_A_01_r}
      boxExtents=<10 10 10>
      constraintSetup {angularLimitsMax=<0 10 0>, angularLimitsMin=<0 -10 -40>, angularTarget=<1 0 0>}
      localJointOffset=<1 0 0>
     }]
   }
   n36 = Anim Dynamics - Bone: dyn_bang_B_01_r(ComponentPose=@n35, Alpha=1, GravityOverride="0,0,0", ExternalForce="0,0,0") {
    angularDampingOverride=0.988
    bOverrideAngularDamping=true
    bUseGravityOverride=true
    simulationSpace=World
    alphaBoolBlend {blendOption=Linear}
    boundBone {boneName=dyn_bang_B_01_r}
    componentAppliedLinearAccClamp=<1500 1500 1500>
    componentLinearAccScale=<1 1 1>
    componentLinearVelScale=<1 1 1>
    physicsBodyDefinitions [{
      boundBone {boneName=dyn_bang_B_01_r}
      boxExtents=<10 10 10>
      constraintSetup {angularLimitsMax=<0 10 0>, angularLimitsMin=<0 -10 -40>, angularTarget=<0 0 0>}
      localJointOffset=<0.8 0 0>
     }]
   }
   n37 = Anim Dynamics - Bone: dyn_bang_C_01_r(ComponentPose=@n36, Alpha=1, GravityOverride="0,0,0", ExternalForce="0,0,0") {
    angularDampingOverride=0.99
    angularSpringConstant=50
    bAngularSpring=true
    bOverrideAngularDamping=true
    bUseGravityOverride=true
    simulationSpace=World
    alphaBoolBlend {blendOption=Linear}
    boundBone {boneName=dyn_bang_C_01_r}
    componentAppliedLinearAccClamp=<1000 1000 1000>
    componentLinearAccScale=<1 1 1>
    componentLinearVelScale=<1 1 1>
    physicsBodyDefinitions [{
      boundBone {boneName=dyn_bang_C_01_r}
      boxExtents=<10 10 10>
      constraintSetup {angularLimitsMax=<0 10 0>, angularLimitsMin=<0 -10 -40>, angularTarget=<1 0 0>}
      localJointOffset=<1 0 0>
     }]
   }
   n38 = Anim Dynamics - Bone: dyn_hair_lock_l(ComponentPose=@n37, Alpha=1, GravityOverride="0,0,0", ExternalForce="0,0,0") {
    angularDampingOverride=0.999
    angularSpringConstant=1000
    bAngularSpring=true
    bOverrideAngularDamping=true
    bUseGravityOverride=true
    simulationSpace=World
    alphaBoolBlend {blendOption=Linear}
    boundBone {boneName=dyn_hair_lock_l}
    componentAppliedLinearAccClamp=<100 100 100>
    componentLinearAccScale=<0.1 0.1 0.1>
    componentLinearVelScale=<0.2 0.2 0.2>
    physicsBodyDefinitions [{
      boundBone {boneName=dyn_hair_lock_l}
      boxExtents=<15 10 10>
      constraintSetup {angularLimitsMax=<0 10 7>, angularLimitsMin=<0 -15 0>, angularTarget=<1 0 0>}
      localJointOffset=<0.5 0 0>
     }]
   }
   n39 = Anim Dynamics - Bone: dyn_hair_lock_r(ComponentPose=@n38, Alpha=1, GravityOverride="0,0,0", ExternalForce="0,0,0") {
    angularDampingOverride=0.999
    angularSpringConstant=3000
    bAngularSpring=true
    bOverrideAngularDamping=true
    bUseGravityOverride=true
    simulationSpace=World
    alphaBoolBlend {blendOption=Linear}
    boundBone {boneName=dyn_hair_lock_r}
    componentAppliedLinearAccClamp=<350 350 350>
    componentLinearAccScale=<0.5 0.5 0.5>
    componentLinearVelScale=<0.7 0.7 0.7>
    physicsBodyDefinitions [{
      boundBone {boneName=dyn_hair_lock_r}
      boxExtents=<20 20 20>
      constraintSetup {angularLimitsMax=<0 10 7>, angularLimitsMin=<0 -8 -5>, angularTarget=<1 0 0>}
      localJointOffset=<1 0 0>
     }]
   }
   n40 = RigidBody!(
    ComponentPose=@n39
    Alpha
    OverrideWorldGravity="0,0,0"
    ExternalForce
    SimSpaceSettings:Sim Space Settings Structure=(WorldAlpha=1,VelocityScaleZ=1,DampingAlpha=1,MaxLinearVelocity=10000,MaxAngularVelocity=10000,MaxLinearAcceleration=10000,MaxAngularAcceleration=10000,ExternalLinearDragV=(X=0,Y=0,Z=0),ExternalLinearVelocity=(X=0,Y=0,Z=0),ExternalAngularVelocity=(X=0,Y=0,Z=0))
   ) -> (Pose) {
    bClampLinearTranslationLimitToRefPose=true
    bOverrideWorldGravity=true
    overridePhysicsAsset=&F_LRG_DodgyOven_FaceAcc_hoodieStrings_FrontEnd_Physics
    simulationSpace=BaseBoneSpace
    alphaBoolBlend {blendOption=Linear}
    baseBoneRef {boneName=spine_05}
    simSpaceSettings {worldAlpha=1}
   }
   n41 = "Transform (Modify) Bone - Bone: dyn_lace_01_r_MSTR"!(ComponentPose=@n40, bAlphaBoolEnabled:bool=True, Rotation="0,0,2") -> (Pose) {
    alphaInputType=Bool
    rotationMode=BMM_Additive
    alphaBoolBlend {blendOption=Linear}
    boneToModify {boneName=dyn_lace_01_r_MSTR}
    rotation {roll=2}
   }
   n42 = "Transform (Modify) Bone - Bone: dyn_lace_01_l_MSTR"!(ComponentPose=@n41, bAlphaBoolEnabled:bool=True, Rotation="0,0,3") -> (Pose) {
    alphaInputType=Bool
    rotationMode=BMM_Additive
    alphaBoolBlend {blendOption=Linear}
    boneToModify {boneName=dyn_lace_01_l_MSTR}
    rotation {roll=3}
   }
   n43 = Component To Local(ComponentPose=@n42)
   n44 = Get bIsInFrontEnd()
   n45 = RigidBody!(
    ComponentPose=Local To Component(LocalPose=Blend Poses by bool(BlendPose_0=@n43, BlendPose_1=@n30, bActiveValue=@n44) {blendType=Linear})
    Alpha
    OverrideWorldGravity="0,0,0"
    ExternalForce
   ) -> (Pose) {
    bOverrideWorldGravity=true
    overridePhysicsAsset=&/F_LRG_DodgyOven_FaceAcc_Hood_Physics
    simulationSpace=BaseBoneSpace
    alphaBoolBlend {blendOption=Linear}
    baseBoneRef {boneName=head}
    simSpaceSettings {dampingAlpha=0, maxAngularAcceleration=32, maxAngularVelocity=8, maxLinearAcceleration=2800, maxLinearVelocity=700, worldAlpha=1}
   }
   n46 = Anim Dynamics - Bone: dyn_ear_l!(ComponentPose=@n45, Alpha=1, GravityOverride:Vector="0,0,0", ExternalForce="0,0,0") -> (Pose) {
    angularDampingOverride=1
    angularSpringConstant=10000
    bAngularSpring=true
    bOverrideAngularDamping=true
    bUseGravityOverride=true
    gravityScale=0.35
    simulationSpace=BoneRelative
    alphaBoolBlend {blendOption=Linear}
    boundBone {boneName=dyn_ear_l}
    componentAppliedLinearAccClamp=<800 800 800>
    componentLinearAccScale=<1 1 1>
    componentLinearVelScale=<1 1 1>
    physicsBodyDefinitions [{
      boundBone {boneName=dyn_ear_l}
      boxExtents=<30 20 20>
      constraintSetup {angularLimitsMax=<1 5 5>, angularLimitsMin=<-1 -5 -5>, angularTarget=<1 0 0>}
      localJointOffset=<1 0.2 0>
     }]
    relativeSpaceBone {boneName=spine_04}
   }
   n47 = Anim Dynamics - Bone: dyn_ear_r!(ComponentPose=@n46, Alpha=1, GravityOverride:Vector="0,0,0", ExternalForce="0,0,0") -> (Pose) {
    angularDampingOverride=1
    angularSpringConstant=10000
    bAngularSpring=true
    bOverrideAngularDamping=true
    bUseGravityOverride=true
    gravityScale=0.35
    simulationSpace=BoneRelative
    alphaBoolBlend {blendOption=Linear}
    boundBone {boneName=dyn_ear_r}
    componentAppliedLinearAccClamp=<1300 1300 1300>
    componentLinearAccScale=<1 1 1>
    componentLinearVelScale=<1 1 1>
    physicsBodyDefinitions [{
      boundBone {boneName=dyn_ear_r}
      boxExtents=<15 20 20>
      constraintSetup {angularLimitsMax=<1 5 10>, angularLimitsMin=<-1 -5 -10>, angularTarget=<1 0 0>}
      localJointOffset=<2 0.5 -0.5>
     }]
    relativeSpaceBone {boneName=spine_04}
   }
   Save cached pose 'DynamicPose'(Pose=Component To Local(ComponentPose=@n47)) {cacheName=DynamicPose}
  ]}
 EventGraph {nodes [
   Event BlueprintUpdateAnimation()
   > n1 = Sequence!(execute) -> (then_0, then_1, then_2:Exec)
   > Set IsNOTRigidBodyEnabled(IsNOTRigidBodyEnabled:bool=NOT Boolean(A=Get bRigidBodyEnabled())) -> (Output_Get:bool)
   > Set IsNOTPlayingEmotes(IsNOTPlayingEmotes:bool=NOT Boolean(A=Get bIsPlayingEmote())) -> (Output_Get:bool)
   n2 = Gravity Override State Function DEPRECATED(
    execute=@n1.then_1
    CharacterPartAnimInstance=Self-Reference()
    GameTimeInSeconds=GetGameTimeInSeconds()
    GravityOverrideParams=Get Gravity Override Params()
    RBAN Params=Get RBAN Params()
   )
   > Set Gravity Override(Gravity Override=@n2)
   > Set RBAN Settings Out(RBAN Settings Out:RBANSetting Struct Structure=@n2.RBAN SettingsOut) -> (Output_Get:RBANSetting Struct Structure)
   n3 = GetOwnerSkeletalMeshComponent()
   n4 = BreakRotator(InRot=BreakTransform(InTransform=GetDeltaTransformFromRefPose(self=@n3, BoneName=spine_04, BaseName=spine_05)).Rotation)
   Set LaceRotation(execute=@n1.then_2, LaceRotation:Rotator=MakeRotator(Roll=MapRangeClamped(Value=@n4.Yaw, InRangeA=0, InRangeB=-10, OutRangeB=10))) -> (Output_Get:Rotator)
   > Set LaceTranslation(LaceTranslation:Vector=MakeVector(Y=MapRangeClamped(Value=@n4.Yaw, InRangeA=0, InRangeB=-10, OutRangeB=2))) -> (Output_Get:Vector)
   n5 = GetOwnerSkeletalMeshComponent()
   n6 = BreakRotator(InRot=BreakTransform(InTransform=GetDeltaTransformFromRefPose(self=@n5, BoneName=spine_04, BaseName=spine_05)).Rotation)
   n7 = MapRangeClamped(Value=@n6.Yaw, InRangeA=0, InRangeB=0, OutRangeB=0)
   > Set LaceRotationFE(LaceRotationFE:Rotator=MakeRotator(Roll=@n7, Yaw=@n7)) -> (Output_Get:Rotator)
   n8 = MapRangeClamped(Value=@n6.Yaw, InRangeA=0, InRangeB=-15)
   > Set LaceTranslationFE(LaceTranslationFE:Vector=MakeVector(X=@n8, Y=@n8)) -> (Output_Get:Vector)
  ]}
]
variables [
 Gravity Override {type=Vector}
 Gravity Override Params {type=GravityOverrideParamsStruct}
 IsNOTPlayingEmotes {type=boolean}
 IsNOTRigidBodyEnabled {type=boolean}
 LaceRotation {type=Rotator}
 LaceRotationFE {type=Rotator}
 LaceTranslation {type=Vector}
 LaceTranslationFE {type=Vector}
 RBAN Params {
  type=RBANParamsStruct
  default {
   dBNO {simSpaceSettings {
     maxAngularAcceleration=1000
     maxAngularVelocity=3
     maxLinearAcceleration=300
     maxLinearVelocity=700
     velocityScaleZ=0
     externalLinearDragV=<0 0.5 0>
    }}
   dBNO_Carried {simSpaceSettings {
     maxAngularAcceleration=1000
     maxAngularVelocity=3
     maxLinearAcceleration=300
     maxLinearVelocity=700
     velocityScaleZ=0
     externalLinearDragV=<0 0.5 0>
    }}
   emoteOrMelee {simSpaceSettings {
     maxAngularAcceleration=1000
     maxAngularVelocity=3
     maxLinearAcceleration=300
     maxLinearVelocity=700
     velocityScaleZ=0
     externalLinearDragV=<0 0.5 0>
    }}
   front_End {simSpaceSettings {
     dampingAlpha=0
     maxAngularAcceleration=30
     maxAngularVelocity=1
     maxLinearAcceleration=100
     maxLinearVelocity=500
     velocityScaleZ=1
     externalLinearDragV=<0 0 0>
    }}
   onGround_Crouch_Moving {simSpaceSettings {
     maxAngularAcceleration=1000
     maxAngularVelocity=3
     maxLinearAcceleration=300
     maxLinearVelocity=700
     velocityScaleZ=0
     externalLinearDragV=<0 0.5 0>
    }}
   onGround_Falling {simSpaceSettings {
     maxAngularAcceleration=1000
     maxAngularVelocity=3
     maxLinearAcceleration=300
     maxLinearVelocity=700
     velocityScaleZ=0
     externalLinearDragV=<0 0.5 0>
    }}
   onGround_Moving {simSpaceSettings {
     maxAngularAcceleration=1000
     maxAngularVelocity=3
     maxLinearAcceleration=300
     maxLinearVelocity=700
     velocityScaleZ=0
     externalLinearDragV=<0 0.5 0>
    }}
   onGround_Standing {simSpaceSettings {
     maxAngularAcceleration=1000
     maxAngularVelocity=3
     maxLinearAcceleration=300
     maxLinearVelocity=700
     velocityScaleZ=0
     externalLinearDragV=<0 0.5 0>
    }}
   rail_Grinding {simSpaceSettings {
     maxAngularAcceleration=1000
     maxAngularVelocity=3
     maxLinearAcceleration=300
     maxLinearVelocity=700
     velocityScaleZ=0
     worldAlpha=0.1
     externalLinearDragV=<0 0.5 0>
    }}
   riding_Moving {simSpaceSettings {
     maxAngularAcceleration=1000
     maxAngularVelocity=3
     maxLinearAcceleration=300
     maxLinearVelocity=700
     velocityScaleZ=0
     externalLinearDragV=<0 0.5 0>
    }}
   riding_Moving_Targeting {simSpaceSettings {
     maxAngularAcceleration=1000
     maxAngularVelocity=3
     maxLinearAcceleration=300
     maxLinearVelocity=700
     velocityScaleZ=0
     externalLinearDragV=<0 0.5 0>
    }}
   riding_Standing {simSpaceSettings {
     maxAngularAcceleration=1000
     maxAngularVelocity=3
     maxLinearAcceleration=300
     maxLinearVelocity=700
     velocityScaleZ=0
     externalLinearDragV=<0 0.5 0>
    }}
   skydive_DivingDown {simSpaceSettings {
     maxAngularAcceleration=1000
     maxAngularVelocity=3
     maxLinearAcceleration=300
     maxLinearVelocity=700
     velocityScaleZ=0
     externalLinearDragV=<0 0.5 0>
    }}
   skydive_FloatingDown {
    componentLinearAccScale=<0 0.5 0>
    componentLinearVelScale=<0 0.5 0>
    simSpaceSettings {
     maxAngularAcceleration=1000
     maxAngularVelocity=3
     maxLinearAcceleration=300
     maxLinearVelocity=700
     velocityScaleZ=0
     externalLinearDragV=<0 0.5 0>
    }
   }
   skydive_GoingUp {simSpaceSettings {
     maxAngularAcceleration=1000
     maxAngularVelocity=3
     maxLinearAcceleration=300
     maxLinearVelocity=700
     velocityScaleZ=0
     externalLinearDragV=<0 0.5 0>
    }}
   skydive_Parachuting {
    componentLinearAccScale=<0 0.1 0.3>
    componentLinearVelScale=<0 0.1 0.3>
    simSpaceSettings {
     maxAngularAcceleration=1000
     maxAngularVelocity=3
     maxLinearAcceleration=300
     maxLinearVelocity=700
     velocityScaleZ=0
     externalLinearDragV=<0 0.5 0>
    }
   }
   skydive_Surfing {simSpaceSettings {
     maxAngularAcceleration=1000
     maxAngularVelocity=3
     maxLinearAcceleration=300
     maxLinearVelocity=700
     velocityScaleZ=0
     externalLinearDragV=<0 0.5 0>
    }}
   swimming {simSpaceSettings {
     maxAngularAcceleration=1000
     maxAngularVelocity=3
     maxLinearAcceleration=300
     maxLinearVelocity=700
     velocityScaleZ=0
     externalLinearDragV=<0 0.5 0>
    }}
   tactical_Sprint {simSpaceSettings {
     maxAngularAcceleration=1000
     maxAngularVelocity=3
     maxLinearAcceleration=300
     maxLinearVelocity=700
     velocityScaleZ=0
     externalLinearDragV=<0 0.5 0>
    }}
   ziplining {simSpaceSettings {
     maxAngularAcceleration=1000
     maxAngularVelocity=3
     maxLinearAcceleration=300
     maxLinearVelocity=700
     velocityScaleZ=0
     externalLinearDragV=<0 0.5 0>
    }}
  }
 }
 RBAN Settings Out {
  type=RBANSettingStruct
  default {
   componentLinearAccScale=<0 0 0>
   componentLinearVelScale=<0 0 0>
   simSpaceSettings {
    dampingAlpha=1
    maxAngularAcceleration=1000
    maxAngularVelocity=3
    maxLinearAcceleration=300
    maxLinearVelocity=700
    velocityScaleZ=0
    worldAlpha=1
    externalAngularVelocity=<0 0 0>
    externalLinearDragV=<0 0.5 0>
    externalLinearVelocity=<0 0 0>
   }
  }
 }
]
